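# Niagara system for the prop spawn effect of the main player's CrossWiper emote.
# It combines seven emitters named as LWE (likely lightweight): sparkles, an airglow, and five directional smoke layers (up, two left, two right).
# Only the SystemState module is recorded, and no emitter internals or renderers are shown.
# A User.NormalizedNotifyProgress float, defaulting to zero, is exposed, presumably for animation notify progress.
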NiagaraSystem /BRCosmetics/Animation/Game/MainPlayer/Emotes/CrossWiper/VFX/NS_CrossWiper_PropSpawn_LWE_1
dependencies {usedDataInterfaces [], usedDynamicInputs [], usedModules [&SystemState]}
emitters {
 AirglowLWE {rendererClasses [], renderers []}
 SmokeLWELeft {rendererClasses [], renderers []}
 SmokeLWELeft001 {rendererClasses [], renderers []}
 SmokeLWERight {rendererClasses [], renderers []}
 SmokeLWERight001 {rendererClasses [], renderers []}
 SmokeLWEUp {rendererClasses [], renderers []}
 SparklesLWE {rendererClasses [], renderers []}
}
summary {
 systemName=NS_CrossWiper_PropSpawn_LWE_1
 emitters [
  {emitterName=SparklesLWE, rendererClasses []}
  {emitterName=AirglowLWE, rendererClasses []}
  {emitterName=SmokeLWEUp, rendererClasses []}
  {emitterName=SmokeLWELeft, rendererClasses []}
  {emitterName=SmokeLWERight, rendererClasses []}
  {emitterName=SmokeLWELeft001, rendererClasses []}
  {emitterName=SmokeLWERight001, rendererClasses []}
 ]
 userVariables [User.NormalizedNotifyProgress {type=NiagaraFloat, defaultValue=0, description=""}]
}
