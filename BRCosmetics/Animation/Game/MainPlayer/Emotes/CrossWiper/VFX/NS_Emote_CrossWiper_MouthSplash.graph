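# Niagara effect for the CrossWiper emote that produces a water splash at the mouth, as its name indicates, spawning from a sampled skeletal mesh location.
# Foam emits short-lived sprites with outward velocity, gravity, drag, SubUV frames and inherited source movement that fades with emitter age.
# Splash_Mesh emits scaling, fading mesh particles animated through curve-driven dynamic material parameters.
# Both tint by time of day, while the Droplets emitter shows no modules or renderers.
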NiagaraSystem /BRCosmetics/Animation/Game/MainPlayer/Emotes/CrossWiper/VFX/NS_Emote_CrossWiper_MouthSplash
dependencies {
 usedDataInterfaces [
  &NiagaraDataInterfaceCurve
  &NiagaraDataInterfaceVectorCurve
  &NiagaraDataInterfaceSpriteRendererInfo
  &NiagaraDataInterfaceSkeletalMesh
  &NiagaraDataInterfaceColorCurve
 ]
 usedDynamicInputs [&MaskVectorBySpawnGroup, &ColorFromCurve]
 usedModules [
  &self:SystemSpawnScript.SystemScriptSource.SystemScriptGraph.NiagaraNodeAssignment_2.SetVariables_#1
  &self:SystemSpawnScript.SystemScriptSource.SystemScriptGraph.NiagaraNodeAssignment_7.SetVariables_#2
  &self:SystemSpawnScript.SystemScriptSource.SystemScriptGraph.NiagaraNodeAssignment_5.SetVariables_#3
  &self:SystemSpawnScript.SystemScriptSource.SystemScriptGraph.NiagaraNodeAssignment_11.SetVariables_#4
  &self:SystemSpawnScript.SystemScriptSource.SystemScriptGraph.NiagaraNodeAssignment_12.SetVariables_#5
  &self:SystemSpawnScript.SystemScriptSource.SystemScriptGraph.NiagaraNodeAssignment_13.SetVariables_#6
  &self:SystemSpawnScript.SystemScriptSource.SystemScriptGraph.NiagaraNodeAssignment_14.SetVariables_#7
  &self:SystemSpawnScript.SystemScriptSource.SystemScriptGraph.NiagaraNodeAssignment_15.SetVariables_#8
  &self:SystemSpawnScript.SystemScriptSource.SystemScriptGraph.NiagaraNodeAssignment_0.SetVariables_#9
  &self:SystemSpawnScript.SystemScriptSource.SystemScriptGraph.NiagaraNodeAssignment_3.SetVariables_#10
  &self:SystemSpawnScript.SystemScriptSource.SystemScriptGraph.NiagaraNodeAssignment_1.SetVariables_#11
  &SystemState
 ]
}
emitters {
 Droplets {rendererClasses [], renderers []}
 Foam {
  emitterUpdateScript [
   &EmitterState(
    Enable Distance Culling=true
    Life Cycle Mode=::NewEnumerator1
    Loop Behavior=::NewEnumerator1
    Loop Delay=0.1
    Loop Duration=1.865
    Max Distance Response=::KillImmediately
    Module.bUseMaxDistance=true
    Module.bUseMinDistance=false
    Scalability Mode=::NewEnumerator1
    UseLoopDelay=true
   )
   &SpawnRate(SpawnRate=15, Use Spawn Probability=false)
  ]
  particleSpawnScript [
   &InitializeParticle(
    Color=$System.Water Splashes Color
    Lifetime=&RandomRangeFloat(Minimum=0.2, Maximum=0.3)
    Mass Max=1
    Mass Min=0.9
    Mass Mode=::NewEnumerator3
    Sprite Rotation Mode=::NewEnumerator2
    Sprite Size Max=<24 24>
    Sprite Size Min=<22 22>
    Sprite Size Mode=::NewEnumerator2
    UsePositionOffset=false
    Write Material Random=true
   )
   /SetVariables() {isSetParameters=true}
   &SkeletalMeshLocation(
    MaskSpawnGroup=false
    Mesh Sampling Type=::NewEnumerator5
    Orientation Sampling=::NewEnumerator1
    Sampled Position Offset=<-5 10 0>
    Skeletal Mesh=$System.Skeletal Mesh
    Socket Sampling Mode=::NewEnumerator5
   )
   &ShapeLocation(
    MaskSpawnGroup=false
    Offset=<0 0 5>
    Offset Coordinate Space=::World
    Offset Mode=::NewEnumerator0
    Override Seed=false
    Shape Origin=$Particles.SkeletalMeshLocation.SampledPosition
    Sphere Radius=3
    UseOwnerScale=true
   )
   &InitialMeshOrientation(Rotation=&RandomRangeVector(Minimum=<-1 -1 0>, Maximum=<1 1 1>), Use Rotation Vector=true)
   &AddVelocity(
    Module.Use Falloff Distance=false
    Origin Offset=<0 -5 -4>
    Override Seed=false
    Velocity Mode=::NewEnumerator1
    Velocity Origin=$Particles.SkeletalMeshLocation.SampledPosition
    Velocity Speed=&RandomRangeFloat(Minimum=60, Maximum=70)
   )
  ]
  particleUpdateScript [
   &GravityForce(Gravity=&RandomRangeVector(Minimum=<0 0 -500>, Maximum=<0 0 -600>))
   &Drag(Drag=&RandomRangeFloat(Minimum=0.9, Maximum=1.1), Use Linear Drag=true, Use Rotational Drag=false)
   &InheritSourceMovement(
    Applied Position Falloff Distance=1000
    Applied Position Scale=&Multiply_VectorByFloat(
     Vector=<1 1 0>
     Float=&FloatFromCurve(
      FloatCurve=DataInterface() {propertyValues {Curve {keys [
          {interpMode=RCIM_Cubic, tangentMode=RCTM_User, time=0.1865, value=1}
          {arriveTangent=-2.1905, interpMode=RCIM_Cubic, leaveTangent=-2.1905, tangentMode=RCTM_User, time=1, value=0}
         ]}}}
      CurveIndex=$Emitter.Age
     )
    )
    ApplyForce=false
    ApplyPosition=true
    ApplyPositionFalloff=false
   )
   &SolveForcesAndVelocity(Clamp Velocity=false, Limit Acceleration=false)
   &ParticleState()
   &ScaleSpriteSize(Uniform Curve Sprite Scale=DataInterface() {propertyValues {Curve {keys [{tangentMode=RCTM_User, time=0, value=0.5}, {tangentMode=RCTM_User, time=1, value=1}]}}})
   &DynamicMaterialParameters(
    Index 0 Param 1=&RandomRangeFloat(Minimum=0, Maximum=0.1)
    Index 0 Param 2=0.15
    Index 0 Param 3=&RandomRangeFloat(Minimum=0.1, Maximum=0.2)
    Index 0 Param 4=&FloatFromCurve(FloatCurve=DataInterface() {propertyValues {Curve {
        defaultValue=1
        keys [
         {arriveTangent=2.3681, interpMode=RCIM_Cubic, leaveTangent=0.0879, tangentMode=RCTM_Break, time=0.5, value=0}
         {arriveTangent=7.4665, interpMode=RCIM_Cubic, leaveTangent=6.456, tangentMode=RCTM_Break, time=1, value=1}
        ]
       }}})
    Module.Index0Float0Write=true
    Module.Index0Float1Write=true
    Module.Index0Float2Write=true
    Module.Index0Float3Write=true
   )
   &SubUVAnimation(
    End Frame Range Override=3
    "Sprite Renderer "=DataInterface() {propertyValues {SpriteRenderer=&self:Foam_0.Renderer}}
    SubUV Animation Mode:ENiagara_SubUVLookupModeV2=::NewEnumerator2
    UseEndFrame=true
    UseStartFrame=true
   )
   &ScaleColor(
    Scale Alpha=&FloatFromCurve(
     FloatCurve=DataInterface() {propertyValues {Curve {keys [
         {arriveTangent=-0.2723, interpMode=RCIM_Cubic, leaveTangent=14.5791, tangentMode=RCTM_Break, time=0, value=0}
         {interpMode=RCIM_Cubic, tangentMode=RCTM_Break, time=0.1, value=1}
         {arriveTangent=0.0378, interpMode=RCIM_Cubic, leaveTangent=0.0378, tangentMode=RCTM_User, time=0.3785, value=1.0282}
         {arriveTangent=-3.7547, interpMode=RCIM_Cubic, leaveTangent=-3.7547, tangentMode=RCTM_Break, time=0.7, value=0}
        ]}}}
     Scale Curve=&FNTimeOfDayScalar(Night Scale=0.7, Morning Scale=0.5, Day Scale=0.6, Evening Scale=0.5)
    )
    Scale RGB=&MakeVectorFromLinearColorRGB(Color=&FNTimeOfDayColor(Night Color=<rgba 1.5 1.5 1.5 0.7>, Morning Color=<rgba 1 1 1 0.6>, Day Color=<rgba 0.7 0.7 0.7 1>, Evening Color=<rgba 0.5 0.5 0.5 0.5>))
    ScaleA=true
    ScaleRGB=true
   )
  ]
  rendererClasses [&NiagaraSpriteRendererProperties]
  renderers [{rendererClass=&NiagaraSpriteRendererProperties}]
 }
 Splash_Mesh {
  emitterUpdateScript [
   &EmitterState(Life Cycle Mode=::NewEnumerator1, Loop Behavior=::NewEnumerator1, Loop Duration=1.78, UseLoopDelay=false)
   &SpawnRate(SpawnRate=20, Use Spawn Probability=false)
  ]
  particleSpawnScript [
   &InitializeParticle(
    Color=&FNTimeOfDayColor(
     Night Color=<rgba 1 0.9111 0.86 1>
     Morning Color=<rgba 1 0.6201 0.474 1>
     Day Color=<rgba 0.9297 0.9297 0.9297 1>
     Evening Color=<rgba 1 0.6201 0.474 1>
    )
    Lifetime=0.75
    Mesh Scale=<0.01 0.01 0.009>
    Mesh Scale Mode=::NewEnumerator4
    Position Mode=::NewEnumerator0
    Write Material Random=true
   ) {module=InitializeParticle001}
   &SkeletalMeshLocation(
    MaskSpawnGroup=false
    Mesh Sampling Type=::NewEnumerator5
    Orientation Sampling=::NewEnumerator1
    Sampled Position Offset=<-9 -4 0>
    Skeletal Mesh=$System.Skeletal Mesh
    Socket Sampling Mode=::NewEnumerator5
   )
   &InitialMeshOrientation(
    Mesh Orientation Mode=::NewEnumerator1
    Rotation=&RandomRangeVector(Minimum=<0 0 0>, Maximum=<0 0 1>)
    Use Rotation Vector=true
   )
   &CameraOffset(Camera Offset Amount=3)
  ]
  particleUpdateScript [
   &ParticleState()
   &DynamicMaterialParameters(
    Index 0 Param 1=&FloatFromCurve(
     FloatCurve=DataInterface() {propertyValues {Curve {keys [{tangentMode=RCTM_User, time=0, value=1}, {tangentMode=RCTM_User, time=1, value=0}]}}}
     Scale Curve=-2
    )
    Index 0 Param 2=&FloatFromCurve(FloatCurve=DataInterface() {propertyValues {Curve {keys [<at -0.0046 4>,<at 1.0011 0>]}}})
    Index 0 Param 3=&FloatFromCurve(
     FloatCurve=DataInterface() {propertyValues {Curve {keys [
         {arriveTangent=4.1785, interpMode=RCIM_Cubic, leaveTangent=4.1785, tangentMode=RCTM_User, time=0, value=0}
         {arriveTangent=0.9995, interpMode=RCIM_Cubic, leaveTangent=0.9995, tangentMode=RCTM_User, time=0.2223, value=0.6825}
         {arriveTangent=-0.0405, interpMode=RCIM_Cubic, leaveTangent=-0.0405, tangentMode=RCTM_User, time=1, value=1}
        ]}}}
     Scale Curve=0.95
    )
    Index 0 Param 4=&FloatFromCurve(
     FloatCurve=DataInterface() {propertyValues {Curve {keys [<at 0.2 0 RCIM_Cubic>, {interpMode=RCIM_Cubic, leaveTangent=3.4641, time=1, value=1}]}}}
     Scale Curve=0.2
    )
    Module.Param0WriteEnabled=true
    Module.Param1WriteEnabled=true
    Module.Param2WriteEnabled=true
    Module.Param3WriteEnabled=true
   ) {module=DynamicMaterialParameters001}
   &ScaleColor(
    Scale Alpha=&FloatFromCurve(FloatCurve=DataInterface() {propertyValues {Curve {keys [{interpMode=RCIM_Cubic, tangentMode=RCTM_User, time=0, value=0.7}, {interpMode=RCIM_Cubic, tangentMode=RCTM_User, time=1, value=0}]}}})
    ScaleA=true
    ScaleRGB=true
   )
   &ScaleMeshSize(Scale Factor=&VectorFromCurve(VectorCurve=DataInterface() {propertyValues {
       XCurve {keys [{arriveTangent=2.6672, time=0, value=2}, <at 0.9989 2.5 RCIM_Cubic>]}
       YCurve {keys [{arriveTangent=2.6672, time=0, value=2}, <at 0.9989 2.5>]}
       ZCurve {keys [
         {arriveTangent=1.1096, interpMode=RCIM_Cubic, leaveTangent=1.1096, tangentMode=RCTM_User, time=0, value=1}
         <at 0.5 1.25 RCIM_Cubic>
         {arriveTangent=-1.0344, interpMode=RCIM_Cubic, leaveTangent=-1.0344, tangentMode=RCTM_User, time=0.9989, value=1}
        ]}
      }})) {module=MeshSizeScale}
  ]
  rendererClasses [&NiagaraMeshRendererProperties]
  renderers [{rendererClass=&NiagaraMeshRendererProperties, rendererIndex=0}]
 }
}
summary {
 systemName=NS_Emote_CrossWiper_MouthSplash
 emitters [
  {emitterName=Droplets, rendererClasses []}
  {emitterName=Splash_Mesh, rendererClasses [&NiagaraMeshRendererProperties]}
  {emitterName=Foam, rendererClasses [&NiagaraSpriteRendererProperties]}
 ]
 userVariables [
  User.bIsFrontEnd {type=NiagaraBool}
  User.bIsFrontEndPreview {type=NiagaraBool}
  User.BurstIndex {type=NiagaraFloat, defaultValue=0, description=""}
  User.FrontEndScale {type=NiagaraFloat, defaultValue=0.4, description=""}
  User.NormalizedNotifyProgress {type=NiagaraFloat, defaultValue=0, description=""}
 ]
}
