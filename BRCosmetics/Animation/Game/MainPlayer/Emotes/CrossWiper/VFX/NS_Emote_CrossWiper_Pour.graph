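# Niagara system for the pouring liquid effect of the CrossWiper player emote.
# Stream001 (ribbons and sprites) and SmallStreams001 (ribbons) form gravity-bent streams, while two Droplets emitters add curl-noise sprite drops that inherit mesh velocity.
# All four emitters sample a shared skeletal mesh reader, using spawn groups to pick bones or sockets.
# Loop-age curves taper the pour, and Stream001 adds platform-dependent spawn rates and time-of-day color scaling.
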NiagaraSystem /BRCosmetics/Animation/Game/MainPlayer/Emotes/CrossWiper/VFX/NS_Emote_CrossWiper_Pour
dependencies {
 usedDataInterfaces [
  &NiagaraDataInterfaceCurve
  &NiagaraDataInterfacePlatformSet
  &NiagaraDataInterfaceSpriteRendererInfo
  &NiagaraDataInterfaceArrayInt32
  &NiagaraDataInterfaceSkeletalMesh
  &NiagaraDataInterfaceColorCurve
 ]
 usedDynamicInputs [
  &ColorFromCurve
  &Waveform
  &Multiply_Vector
 ]
 usedModules [
  &self:SystemSpawnScript.SystemScriptSource.SystemScriptGraph.NiagaraNodeAssignment_0.SetVariables_#1
  &self:SystemSpawnScript.SystemScriptSource.SystemScriptGraph.NiagaraNodeAssignment_4.SetVariables_#2
  &self:SystemSpawnScript.SystemScriptSource.SystemScriptGraph.NiagaraNodeAssignment_6.SetVariables_#3
  &self:SystemSpawnScript.SystemScriptSource.SystemScriptGraph.NiagaraNodeAssignment_7.SetVariables_#4
  &self:SystemSpawnScript.SystemScriptSource.SystemScriptGraph.NiagaraNodeAssignment_20.SetVariables_#5
  &self:SystemSpawnScript.SystemScriptSource.SystemScriptGraph.NiagaraNodeAssignment_19.SetVariables_#6
  &self:SystemSpawnScript.SystemScriptSource.SystemScriptGraph.NiagaraNodeAssignment_18.SetVariables_#7
  &self:SystemSpawnScript.SystemScriptSource.SystemScriptGraph.NiagaraNodeAssignment_8.SetVariables_#8
  &self:SystemSpawnScript.SystemScriptSource.SystemScriptGraph.NiagaraNodeAssignment_1.SetVariables_#9
  &self:SystemSpawnScript.SystemScriptSource.SystemScriptGraph.NiagaraNodeAssignment_5.SetVariables_#10
  &self:SystemSpawnScript.SystemScriptSource.SystemScriptGraph.NiagaraNodeAssignment_9.SetVariables_#11
  &self:SystemSpawnScript.SystemScriptSource.SystemScriptGraph.NiagaraNodeAssignment_10.SetVariables_#12
  &self:SystemSpawnScript.SystemScriptSource.SystemScriptGraph.NiagaraNodeAssignment_11.SetVariables_#13
  &SystemState
  &self:SystemSpawnScript.SystemScriptSource.SystemScriptGraph.NiagaraNodeAssignment_2.SetVariables_#14
  &self:SystemSpawnScript.SystemScriptSource.SystemScriptGraph.NiagaraNodeAssignment_3.SetVariables_#15
 ]
}
emitters {
 Droplets003 {
  emitterUpdateScript [
   &EmitterState(
    Enable Distance Culling=true
    Life Cycle Mode=::NewEnumerator1
    Loop Behavior=::NewEnumerator1
    Loop Delay=0.01
    Max Distance Response=::KillImmediately
    Module.bUseMaxDistance=true
    Module.bUseMinDistance=false
    Scalability Mode=::NewEnumerator1
    UseLoopDelay=true
   )
   &SpawnRate(
    Spawn Probability=0.8
    SpawnGroup=1
    SpawnRate=&FloatFromCurve(
     FloatCurve=DataInterface() {propertyValues {Curve {
        defaultValue=1
        keys [
         {tangentMode=RCTM_User, time=0, value=1}
         {arriveTangent=-0.0405, interpMode=RCIM_Cubic, leaveTangent=-0.0405, tangentMode=RCTM_User, time=0.7, value=1}
         {arriveTangent=-4.5714, interpMode=RCIM_Cubic, leaveTangent=-4.5714, tangentMode=RCTM_User, time=1, value=0.5}
        ]
       }}}
     CurveIndex=$Emitter.NormalizedLoopAge
     Scale Curve=40
    )
    Use Spawn Probability=false
   ) {module=SpawnRate001}
   &SpawnRate(
    Spawn Probability=0.8
    SpawnRate=&FloatFromCurve(
     FloatCurve=DataInterface() {propertyValues {Curve {
        defaultValue=1
        keys [
         {tangentMode=RCTM_User, time=0, value=1}
         {arriveTangent=-0.0405, interpMode=RCIM_Cubic, leaveTangent=-0.0405, tangentMode=RCTM_User, time=0.7, value=1}
         {arriveTangent=-4.5714, interpMode=RCIM_Cubic, leaveTangent=-4.5714, tangentMode=RCTM_User, time=1, value=0.5}
        ]
       }}}
     CurveIndex=$Emitter.NormalizedLoopAge
     Scale Curve=40
    )
    Use Spawn Probability=false
   )
  ]
  particleSpawnScript [
   &InitializeParticle(
    Color=$System.DropletColors
    Lifetime Max=1.3
    Lifetime Min=0.7
    Lifetime Mode=::NewEnumerator1
    Mass Max=1
    Mass Min=0.9
    Mass Mode=::NewEnumerator3
    Sprite Rotation Mode=::NewEnumerator2
    Sprite Size=&Multiply_Vector2D_ByFloat(
     Float=&RandomRangeFloat(Minimum=0.9, Maximum=1)
     Vector2D=&RandomRangeVector2D(Minimum=<1 1.4>, Maximum=<1.1 1.5>)
    )
    Sprite Size Mode=::NewEnumerator4
    UsePositionOffset=false
    Write Material Random=true
   )
   /SetVariables(Particles.Random=&UniformAOrBFloat(A=-1)) {isSetParameters=true, module=SetVariables_#16}
   &SkeletalMeshLocation(
    Bone / Socket Index=&SelectIntFromArray(Direct Array Index=$Emitter.SpawnGroup, Int Selection Array=DataInterface() {propertyValues {IntData [2,3]}})
    Bone Sampling Mode=::NewEnumerator4
    "Inherit Velocity (Scale)":NiagaraFloat=&FloatFromCurve(
     FloatCurve=DataInterface() {propertyValues {Curve {
        defaultValue=1
        keys [{interpMode=RCIM_Cubic, tangentMode=RCTM_User, time=0.6, value=0.5}, {interpMode=RCIM_Cubic, tangentMode=RCTM_User, time=0.7, value=1}]
       }}}
     CurveIndex=$Emitter.NormalizedLoopAge
     Scale Curve=&RandomRangeFloat(Minimum=1, Maximum=2)
    )
    Inherited Velocity Speed Limit=500
    MaskSpawnGroup=false
    Orientation Sampling=::NewEnumerator1
    Sampled Position Offset=<2 0 5>
    Skeletal Mesh=$System.Skeletal Mesh Reader
    Velocity Sampling=::NewEnumerator0
    VelocitySpeedLimit=true
   )
   &ShapeLocation(
    Additional Quaternion Rotation=false
    Cylinder Height=17
    Cylinder Radius=3
    MaskSpawnGroup=false
    Override Seed=false
    Rotation Mode=::NewEnumerator3
    Rotation Quaternion=$Particles.SkeletalMeshLocation.SampledRotation
    Shape Origin=$Particles.SkeletalMeshLocation.SampledPosition
    Shape Primitive=::NewEnumerator1
    Surface Only=false
    UseOwnerScale=true
   )
   &CurlNoiseForce(Bias Noise=false, Noise Frequency=200, Noise Strength=50, Pan Noise=false)
   &ApplyInitialForces()
   /SetVariables(Particles.SpriteSize=&Multiply_Vector2D_ByFloat(
     Float=&FloatFromCurve(
      FloatCurve=DataInterface() {propertyValues {Curve {defaultValue=1, keys [{tangentMode=RCTM_User, time=0, value=1}, {tangentMode=RCTM_User, time=1, value=1}]}}}
      CurveIndex=$Emitter.NormalizedLoopAge
      Scale Curve=1.25
     )
     Vector2D=$Particles.SpriteSize
    )) {isSetParameters=true}
  ]
  particleUpdateScript [
   &ParticleState()
   &GravityForce()
   &Drag(Drag=1.25, Use Linear Drag=true, Use Rotational Drag=false)
   &SolveForcesAndVelocity(
    Clamp Velocity=false
    Limit Acceleration=false
    Write to Intrinsic Properties:NiagaraBool=&SetBoolByFloatComparison(A=$Particles.NormalizedAge, B=0.25)
   )
   &ScaleSpriteSize(Uniform Curve Sprite Scale=DataInterface() {propertyValues {Curve {keys [
        <at 0 0 RCIM_Cubic>
        {arriveTangent=0.3157, leaveTangent=0.3157, time=0.19, value=1}
        {arriveTangent=-4.5487, interpMode=RCIM_Cubic, leaveTangent=-4.5487, tangentMode=RCTM_User, time=0.5, value=1}
        <at 1 0 RCIM_Cubic>
       ]}}}) {module=ScaleSpriteSize001}
   &ScaleSpriteSizeBySpeed(Max Scale Factor=<0.7 3>, Min Scale Factor=<0.9 1>, Velocity Threshold=300)
   &ScaleColor(
    Scale Alpha=&FloatFromCurve(FloatCurve=DataInterface() {propertyValues {Curve {keys [<at 0 1>,<at 0.7 1>,<at 1 0>]}}}, Scale Curve=0.75)
    ScaleA=true
    ScaleRGB=false
   )
  ]
  rendererClasses [&NiagaraSpriteRendererProperties]
  renderers [{rendererClass=&NiagaraSpriteRendererProperties}]
 }
 Droplets004 {
  emitterUpdateScript [
   &EmitterState(
    Enable Distance Culling=true
    Life Cycle Mode=::NewEnumerator1
    Loop Behavior=::NewEnumerator1
    Loop Duration=0.7
    Max Distance Response=::KillImmediately
    Module.bUseMaxDistance=true
    Module.bUseMinDistance=false
    Scalability Mode=::NewEnumerator1
    UseLoopDelay=true
   )
   &SpawnRate(
    Spawn Probability=0.8
    SpawnGroup=1
    SpawnRate=&FloatFromCurve(
     FloatCurve=DataInterface() {propertyValues {Curve {
        defaultValue=1
        keys [
         {tangentMode=RCTM_User, time=0, value=1}
         {arriveTangent=-0.0405, interpMode=RCIM_Cubic, leaveTangent=-0.0405, tangentMode=RCTM_User, time=0.7009, value=1}
         {arriveTangent=-8.0089, interpMode=RCIM_Cubic, leaveTangent=-8.0089, tangentMode=RCTM_User, time=1, value=0}
        ]
       }}}
     CurveIndex=$Emitter.NormalizedLoopAge
     Scale Curve=160
    )
    Use Spawn Probability=false
   ) {module=SpawnRate001}
   &SpawnRate(
    Spawn Probability=0.8
    SpawnRate=&FloatFromCurve(
     FloatCurve=DataInterface() {propertyValues {Curve {
        defaultValue=1
        keys [
         {tangentMode=RCTM_User, time=0, value=1}
         {arriveTangent=-0.0405, interpMode=RCIM_Cubic, leaveTangent=-0.0405, tangentMode=RCTM_User, time=0.7009, value=1}
         {arriveTangent=-8.0089, interpMode=RCIM_Cubic, leaveTangent=-8.0089, tangentMode=RCTM_User, time=1, value=0}
        ]
       }}}
     CurveIndex=$Emitter.NormalizedLoopAge
     Scale Curve=160
    )
    Use Spawn Probability=false
   )
  ]
  particleSpawnScript [
   &InitializeParticle(
    Color=$System.DropletColors
    Lifetime Max=0.8
    Lifetime Min=0.4
    Lifetime Mode=::NewEnumerator1
    Mass Max=1
    Mass Min=0.9
    Mass Mode=::NewEnumerator3
    Sprite Rotation Mode=::NewEnumerator2
    Sprite Size=&Multiply_Vector2D_ByFloat(
     Float=&UniformAOrBFloat(A=&RandomRangeFloat(Minimum=0.7, Maximum=0.9), B=1.1, Distribution Weight=0.1)
     Vector2D=&RandomRangeVector2D(Minimum=<1 1.5>, Maximum=<1 2>)
    )
    Sprite Size Mode=::NewEnumerator4
    UsePositionOffset=false
    Write Material Random=true
   )
   /SetVariables(Particles.Random=&UniformAOrBFloat(A=-1)) {isSetParameters=true, module=SetVariables_#16}
   &SkeletalMeshLocation(
    Bone / Socket Index=&SelectIntFromArray(Direct Array Index=$Emitter.SpawnGroup, Int Selection Array=DataInterface() {propertyValues {IntData [2,3]}})
    Bone Sampling Mode=::NewEnumerator4
    "Inherit Velocity (Scale)"=1.2
    MaskSpawnGroup=false
    Orientation Sampling=::NewEnumerator1
    Sampled Position Offset=<2 0 5>
    Skeletal Mesh=$System.Skeletal Mesh Reader
    Velocity Sampling=::NewEnumerator0
    VelocitySpeedLimit=true
   )
   &ShapeLocation(MaskSpawnGroup=false, Override Seed=false, Shape Origin=$Particles.SkeletalMeshLocation.SampledPosition, Sphere Radius=2, UseOwnerScale=true)
   &CurlNoiseForce(Bias Noise=false, Noise Frequency=200, Noise Strength=50, Pan Noise=false)
   &ApplyInitialForces()
  ]
  particleUpdateScript [
   &ParticleState()
   &GravityForce()
   &Drag(Drag=1.25, Use Linear Drag=true, Use Rotational Drag=false)
   &SolveForcesAndVelocity(Clamp Velocity=false, Limit Acceleration=false)
   &ScaleSpriteSize(Uniform Curve Sprite Scale=DataInterface() {propertyValues {Curve {keys [{arriveTangent=-4.5487, interpMode=RCIM_Cubic, leaveTangent=-4.5487, tangentMode=RCTM_User, time=0.33, value=1}, <at 1 0 RCIM_Cubic>]}}}) {module=ScaleSpriteSize001}
   &ScaleSpriteSizeBySpeed(Max Scale Factor=<0.9 3>, Min Scale Factor=<0.7 2>, Velocity Threshold=500)
   &ScaleColor(
    Scale Alpha=&FloatFromCurve(FloatCurve=DataInterface() {propertyValues {Curve {keys [<at 0 1>,<at 0.7 1>,<at 1 0>]}}})
    ScaleA=true
    ScaleRGB=false
   )
  ]
  rendererClasses [&NiagaraSpriteRendererProperties]
  renderers [{rendererClass=&NiagaraSpriteRendererProperties}]
 }
 SmallStreams001 {
  emitterUpdateScript [
   &EmitterState(
    Inactive Response=::NewEnumerator1
    Life Cycle Mode=::NewEnumerator1
    Loop Behavior=::NewEnumerator1
    Loop Delay=0.31
    Loop Duration=3.3
    UseLoopDelay=true
   )
   &SpawnRate(SpawnRate=40, Use Spawn Probability=false)
   &SpawnRate(SpawnGroup=1, SpawnRate=40, Use Spawn Probability=false) {module=SpawnRate001}
   &SpawnRate(SpawnGroup=2, SpawnRate=40, Use Spawn Probability=false) {module=SpawnRate002}
   &SpawnRate(SpawnGroup=3, SpawnRate=40, Use Spawn Probability=false) {module=SpawnRate003}
  ]
  particleSpawnScript [
   &InitializeParticle(
    Lifetime=&Multiply_Float(
     A=&FloatFromCurve(
      FloatCurve=DataInterface() {propertyValues {Curve {
         defaultValue=0.7
         keys [
          {tangentMode=RCTM_User, time=0, value=1}
          {arriveTangent=-0.0405, interpMode=RCIM_Cubic, leaveTangent=-0.0405, tangentMode=RCTM_User, time=0.7009, value=1}
          {arriveTangent=-8.0089, interpMode=RCIM_Cubic, leaveTangent=-8.0089, tangentMode=RCTM_User, time=1, value=0}
         ]
        }}}
      CurveIndex=$Emitter.NormalizedLoopAge
     )
     B=&MergeHelper_FloatSwitch(0=0.26, 1=0.3, 2=0.25, 3=0.28, Index=$Emitter.SpawnGroup, Num=::NewEnumerator2)
    )
    Mesh Scale=<3.5 15 1>
    Mesh Scale Mode=::NewEnumerator4
    Ribbon Width=&FloatFromCurve(
     FloatCurve=DataInterface() {propertyValues {Curve {
        defaultValue=1
        keys [
         {tangentMode=RCTM_User, time=0, value=1}
         {arriveTangent=-0.0405, interpMode=RCIM_Cubic, leaveTangent=-0.0405, tangentMode=RCTM_User, time=0.7009, value=1}
         {arriveTangent=-8.0089, interpMode=RCIM_Cubic, leaveTangent=-8.0089, tangentMode=RCTM_User, time=1, value=0}
        ]
       }}}
     CurveIndex=$Emitter.NormalizedLoopAge
     Scale Curve=&RandomRangeFloat(Minimum=0.8, Maximum=0.9)
    )
    Ribbon Width Mode=::NewEnumerator1
    Sprite Rotation Angle=90
    Sprite Rotation Mode=::NewEnumerator0
    Sprite Size=<3.5 10>
    Sprite Size Mode=::NewEnumerator4
    UsePositionOffset=false
    Write Material Random=true
   )
   &SkeletalMeshLocation(
    Bone / Socket Index=$Emitter.SpawnGroup
    Bone Sampling Mode=::NewEnumerator4
    MaskSpawnGroup=false
    Orientation Sampling=::NewEnumerator1
    Sampled Position Coordinate Space=::Local
    Sampled Position Offset=&MaskVectorBySpawnGroup(
     Masked Vector=<0 0 -3>
     Passthrough Vector=&MaskVectorBySpawnGroup(
      Masked Vector=<0 0 -3>
      Passthrough Vector=&MaskVectorBySpawnGroup(Masked Vector=<0 0 0.5>, Passthrough Vector=<0 0 0.5>, Spawn Group Mask=2)
      Spawn Group Mask=1
     )
    )
    Skeletal Mesh=$System.Skeletal Mesh Reader
   )
   &AddVelocity(
    Velocity=&ReturnMeshOrientationAxis(Mesh Orientation Axis=::XAxis, Mesh Orientation Quat=$Particles.SkeletalMeshLocation.SampledRotation)
    Velocity Speed Scale=&FloatFromCurve(
     FloatCurve=DataInterface() {propertyValues {Curve {
        defaultValue=1.2
        keys [
         {tangentMode=RCTM_User, time=0, value=1}
         {arriveTangent=-0.0405, interpMode=RCIM_Cubic, leaveTangent=-0.0405, tangentMode=RCTM_User, time=0.7009, value=1}
         {arriveTangent=-8.0089, interpMode=RCIM_Cubic, leaveTangent=-8.0089, tangentMode=RCTM_User, time=1, value=0}
        ]
       }}}
     Scale Curve=50
    )
   )
   &ApplyInitialForces()
   /SetVariables(Particles.RandPerSpawnGroup:NiagaraFloat=/Game/Effects/Niagara/DynamicInputs/Random/FN_SeededRandomRangeFloat(Minimum:NiagaraFloat=0.1, Maximum:NiagaraFloat=1, Random Seed:NiagaraInt32=$Emitter.SpawnGroup)) {isSetParameters=true}
  ]
  particleUpdateScript [
   &ParticleState()
   &GravityForce()
   &SolveForcesAndVelocity(Clamp Velocity=false, Limit Acceleration=false)
   &DynamicMaterialParameters(
    Index 0 Param 1=$Particles.Initial.RandPerSpawnGroup
    Index 0 Param 2=&FloatFromCurve(FloatCurve=DataInterface() {propertyValues {Curve {keys [
         <at 0.0009 0.4>
         {arriveTangent=-1.1863, interpMode=RCIM_Cubic, leaveTangent=-1.1863, tangentMode=RCTM_User, time=0.4, value=0.4}
         {arriveTangent=0.0526, interpMode=RCIM_Cubic, tangentMode=RCTM_Break, time=1, value=0}
        ]}}})
    Module.Index0Float0Write=true
    Module.Index0Float1Write=true
    Module.Index0Float2Write=true
    Module.Index0Float3Write=true
   )
   &ScaleColor(
    Scale Alpha=&FloatFromCurve(
     FloatCurve=DataInterface() {propertyValues {Curve {
        defaultValue=1
        keys [
         {tangentMode=RCTM_User, time=0, value=1}
         {arriveTangent=-0.0405, interpMode=RCIM_Cubic, leaveTangent=-0.0405, tangentMode=RCTM_User, time=0.7009, value=1}
         {arriveTangent=-8.0089, interpMode=RCIM_Cubic, leaveTangent=-8.0089, tangentMode=RCTM_User, time=1, value=0}
        ]
       }}}
     Scale Curve=&RandomRangeFloat(Minimum=0.9, Maximum=1)
    )
    ScaleA=true
    ScaleRGB=true
   )
   &ScaleRibbonWidth(Ribbon Width Scale=&FloatFromCurve(
     FloatCurve=DataInterface() {propertyValues {Curve {
        defaultValue=1
        keys [{interpMode=RCIM_Cubic, tangentMode=RCTM_User, time=0, value=1}, {interpMode=RCIM_Cubic, tangentMode=RCTM_User, time=1, value=3}]
       }}}
     Scale Curve=2
    ))
   &SubUVAnimation(
    End Frame Range Override=3
    "Sprite Renderer "=DataInterface() {propertyValues {SpriteRenderer=&self:SmallStreams001_0.Renderer}}
    SubUV Animation Mode:ENiagara_SubUVLookupModeV2=::NewEnumerator2
    UseEndFrame=true
    UseStartFrame=true
   )
   /SetVariables(
    Particles.DropletTrailWidth:NiagaraFloat=&Multiply_Float(
     A=$Particles.Initial.RibbonWidth
     B=&FloatFromCurve(
      FloatCurve=DataInterface() {propertyValues {Curve {
         defaultValue=1
         keys [
          {arriveTangent=11.3245, interpMode=RCIM_Cubic, leaveTangent=11.3245, tangentMode=RCTM_User, time=0, value=1}
          {interpMode=RCIM_Cubic, tangentMode=RCTM_User, time=1, value=6.5}
         ]
        }}}
      Scale Curve=3
     )
    )
    Particles.Dyn1:Vector4f=&MakeVector4(
     X=$Particles.Initial.RandPerSpawnGroup
     Y=&FloatFromCurve(FloatCurve=DataInterface() {propertyValues {Curve {keys [
          <at 0 0.5>
          {arriveTangent=-1.1863, interpMode=RCIM_Cubic, leaveTangent=-1.1863, tangentMode=RCTM_User, time=0.1, value=0.5}
          {interpMode=RCIM_Cubic, leaveTangent=-2.5488, tangentMode=RCTM_Break, time=0.788, value=0.275}
          {arriveTangent=0.0526, interpMode=RCIM_Cubic, tangentMode=RCTM_Break, time=1, value=0}
         ]}}})
     Z=&RandomRangeFloat(Minimum=0.1, Maximum=0.2)
     W=&FloatFromCurve(FloatCurve=DataInterface() {propertyValues {Curve {
         defaultValue=1
         keys [{interpMode=RCIM_Cubic, tangentMode=RCTM_User, time=0, value=0}, {interpMode=RCIM_Cubic, tangentMode=RCTM_User, time=1, value=1}]
        }}})
    )
    Particles.TrailColor:LinearColor=&MultiplyLinearColor(
     Linear Color=$Particles.Initial.Color
     Scale Factor=&MakeVector4FromVectorAndFloat(
      "Vector (RGB)"=<1 1 1>
      "Float (Alpha)"=&FloatFromCurve(FloatCurve=DataInterface() {propertyValues {Curve {
          defaultValue=1
          keys [
           <at 0.15 1>
           {arriveTangent=-0.0405, interpMode=RCIM_Cubic, leaveTangent=-0.0405, tangentMode=RCTM_User, time=0.7009, value=1}
           {arriveTangent=-8.0089, interpMode=RCIM_Cubic, leaveTangent=-8.0089, tangentMode=RCTM_User, time=1, value=0}
          ]
         }}})
     )
    )
   ) {isSetParameters=true}
  ]
  rendererClasses [&NiagaraRibbonRendererProperties]
  renderers [
   {rendererClass=&NiagaraRibbonRendererProperties}
   {rendererClass=&NiagaraRibbonRendererProperties, rendererIndex=1}
  ]
 }
 Stream001 {
  emitterUpdateScript [
   &EmitterState(
    Inactive Response=::NewEnumerator1
    Life Cycle Mode=::NewEnumerator1
    Loop Behavior=::NewEnumerator1
    Loop Delay=0.88
    Loop Duration=3.1
    UseLoopDelay=true
   )
   &SpawnRate(
    SpawnRate=&FloatFromCurve(
     FloatCurve=DataInterface() {propertyValues {Curve {defaultValue=1, keys [<at 0 1 RCIM_Constant>,<at 0.8 0 RCIM_Constant>]}}}
     CurveIndex=$Emitter.NormalizedLoopAge
     Scale Curve=&FN_PickFloatDependOnPlatform(TargetPlatformActive=60, TargetPlatformNoActive=30)
    )
    Use Spawn Probability=false
   )
   &SpawnRate(
    SpawnGroup=1
    SpawnRate=&FloatFromCurve(
     FloatCurve=DataInterface() {propertyValues {Curve {defaultValue=1, keys [<at 0 1 RCIM_Constant>,<at 0.8 0 RCIM_Constant>]}}}
     CurveIndex=$Emitter.NormalizedLoopAge
     Scale Curve=&FN_PickFloatDependOnPlatform(TargetPlatformActive=60, TargetPlatformNoActive=30)
    )
    Use Spawn Probability=false
   ) {module=SpawnRate001}
   &SpawnRate(
    SpawnRate=&FN_PickFloatDependOnPlatform(TargetPlatformActive=20, TargetPlatformNoActive=0)
    Use Spawn Probability=false
   ) {module=SpawnRate002}
   &SpawnRate(
    SpawnGroup=1
    SpawnRate=&FN_PickFloatDependOnPlatform(TargetPlatformActive=20, TargetPlatformNoActive=0)
    Use Spawn Probability=false
   ) {module=SpawnRate003}
  ]
  particleSpawnScript [
   &InitializeParticle(
    Color=&MergeHelper_ColorSwitch(
     0=<rgba 1 1 1 1>
     1=<rgba 1 1 1 1>
     2=$System.Water Splashes Color
     3=$System.Water Splashes Color
     Index=&SpawnGroup()
     Num=::NewEnumerator2
    )
    Lifetime=&MergeHelper_FloatSwitch(
     0=0.25
     1=0.25
     2=&RandomRangeFloat(Minimum=0.15, Maximum=0.18)
     3=&RandomRangeFloat(Minimum=0.15, Maximum=0.18)
     Index=&SpawnGroup()
     Num=::NewEnumerator2
    )
    Mesh Scale=<3.5 15 1>
    Mesh Scale Mode=::NewEnumerator4
    Ribbon Width=&Multiply_Float(
     A=&RandomRangeFloat(Minimum=5.5, Maximum=6)
     B=&FloatFromCurve(
      FloatCurve=DataInterface() {propertyValues {Curve {
         defaultValue=1
         keys [
          {tangentMode=RCTM_User, time=0, value=1}
          {arriveTangent=-0.0405, interpMode=RCIM_Cubic, leaveTangent=-0.0405, tangentMode=RCTM_User, time=0.2817, value=1}
          {arriveTangent=-8.0089, interpMode=RCIM_Cubic, leaveTangent=-8.0089, tangentMode=RCTM_User, time=0.391, value=0.5}
         ]
        }}}
      CurveIndex=$Emitter.NormalizedLoopAge
     )
    )
    Ribbon Width Mode=::NewEnumerator1
    Sprite Rotation Angle=90
    Sprite Rotation Mode=::NewEnumerator0
    Sprite Size Mode=::NewEnumerator1
    Uniform Sprite Size Max=2
    Uniform Sprite Size Min=1.8
    UsePositionOffset=false
    Write Material Random=true
   )
   &SkeletalMeshLocation(
    Bone / Socket Index=$Emitter.SpawnGroup
    Bone Sampling Mode=::NewEnumerator4
    MaskSpawnGroup=false
    Orientation Sampling=::NewEnumerator1
    Sampled Position Offset=<0 0 5>
    Skeletal Mesh=$System.Skeletal Mesh Reader
   )
   &AddVelocity(
    Velocity=&ReturnMeshOrientationAxis(Mesh Orientation Axis=::ZAxis, Mesh Orientation Quat=$Particles.SkeletalMeshLocation.SampledRotation)
    Velocity Speed Scale=&FloatFromCurve(FloatCurve=DataInterface() {propertyValues {Curve {keys [<at 0 1.2>,<at 1 1>]}}}, Scale Curve=50)
   )
   &ApplyInitialForces()
   /SetVariables(Particles.VisibilityTag=&MergeHelper_IntSwitch(0=0, 1=0, 2=1, 3=1, 4=?, 5=?, Index=$Emitter.SpawnGroup, Num=::NewEnumerator2)) {isSetParameters=true}
  ]
  particleUpdateScript [
   &ParticleState()
   &GravityForce()
   &SolveForcesAndVelocity(Clamp Velocity=false, Limit Acceleration=false)
   &DynamicMaterialParameters(
    Index 0 Param 1=1
    Index 0 Param 2=&FloatFromCurve(
     FloatCurve=DataInterface() {propertyValues {Curve {
        defaultValue=0.9
        keys [
         {tangentMode=RCTM_User, time=0, value=1}
         {arriveTangent=-0.0405, interpMode=RCIM_Cubic, leaveTangent=-0.0405, tangentMode=RCTM_User, time=0.5, value=1}
         {arriveTangent=-2.2036, interpMode=RCIM_Cubic, leaveTangent=-2.2036, tangentMode=RCTM_User, time=1, value=0}
        ]
       }}}
     Scale Curve=&FloatFromCurve(
      FloatCurve=DataInterface() {propertyValues {Curve {
         defaultValue=0.9
         keys [
          {arriveTangent=-0.3469, interpMode=RCIM_Cubic, leaveTangent=-0.3469, tangentMode=RCTM_User, time=0.5, value=0.8}
          {arriveTangent=-3.9032, interpMode=RCIM_Cubic, leaveTangent=-3.9032, tangentMode=RCTM_User, time=1, value=0}
         ]
        }}}
      CurveIndex=$Emitter.NormalizedLoopAge
     )
    )
    Module.Index0Float0Write=true
    Module.Index0Float1Write=true
    Module.Index0Float2Write=true
    Module.Index0Float3Write=true
   )
   &ScaleColor(
    Scale Alpha=&MergeHelper_FloatSwitch(
     0=&FloatFromCurve(
      FloatCurve=DataInterface() {propertyValues {Curve {keys [
          <at 0 1>
          {interpMode=RCIM_Cubic, leaveTangent=-2.932, tangentMode=RCTM_Break, time=0.5, value=0.9559}
          {arriveTangent=-0.2719, interpMode=RCIM_Cubic, tangentMode=RCTM_Break, time=1, value=0.5}
         ]}}}
      Scale Curve=&RandomRangeFloat(Minimum=0.9, Maximum=1)
     )
     1=&FloatFromCurve(
      FloatCurve=DataInterface() {propertyValues {Curve {keys [
          <at 0 1>
          {interpMode=RCIM_Cubic, leaveTangent=-2.932, tangentMode=RCTM_Break, time=0.5, value=0.9559}
          {arriveTangent=-0.2719, interpMode=RCIM_Cubic, tangentMode=RCTM_Break, time=1, value=0.5}
         ]}}}
      Scale Curve=&RandomRangeFloat(Minimum=0.9, Maximum=1)
     )
     2=&FloatFromCurve(
      FloatCurve=DataInterface() {propertyValues {Curve {keys [
          {arriveTangent=-0.2723, interpMode=RCIM_Cubic, leaveTangent=14.5791, tangentMode=RCTM_Break, time=0, value=0}
          {interpMode=RCIM_Cubic, tangentMode=RCTM_Break, time=0.1, value=1}
          {arriveTangent=0.0378, interpMode=RCIM_Cubic, leaveTangent=0.0378, tangentMode=RCTM_User, time=0.3785, value=1.0282}
          {arriveTangent=-3.7547, interpMode=RCIM_Cubic, leaveTangent=-3.7547, tangentMode=RCTM_Break, time=0.7, value=0}
         ]}}}
      Scale Curve=&FNTimeOfDayScalar(Night Scale=0.7, Morning Scale=0.5, Day Scale=0.6, Evening Scale=0.5)
     )
     3=&FloatFromCurve(
      FloatCurve=DataInterface() {propertyValues {Curve {keys [
          {arriveTangent=-0.2723, interpMode=RCIM_Cubic, leaveTangent=14.5791, tangentMode=RCTM_Break, time=0, value=0}
          {interpMode=RCIM_Cubic, tangentMode=RCTM_Break, time=0.1, value=1}
          {arriveTangent=0.0378, interpMode=RCIM_Cubic, leaveTangent=0.0378, tangentMode=RCTM_User, time=0.3785, value=1.0282}
          {arriveTangent=-3.7547, interpMode=RCIM_Cubic, leaveTangent=-3.7547, tangentMode=RCTM_Break, time=0.7, value=0}
         ]}}}
      Scale Curve=&FNTimeOfDayScalar(Night Scale=0.7, Morning Scale=0.5, Day Scale=0.6, Evening Scale=0.5)
     )
     Index=&SpawnGroup()
     Num=::NewEnumerator2
    )
    Scale RGB=&MergeHelper_VectorSwitch(
     0=<1 1 1>
     1=<1 1 1>
     2=&MakeVectorFromLinearColorRGB(Color=&FNTimeOfDayColor(Night Color=<rgba 1.5 1.5 1.5 0.7>, Morning Color=<rgba 1 1 1 0.6>, Day Color=<rgba 0.7 0.7 0.7 1>, Evening Color=<rgba 0.5 0.5 0.5 0.5>))
     3=&MakeVectorFromLinearColorRGB(Color=&FNTimeOfDayColor(Night Color=<rgba 1.5 1.5 1.5 0.7>, Morning Color=<rgba 1 1 1 0.6>, Day Color=<rgba 0.7 0.7 0.7 1>, Evening Color=<rgba 0.5 0.5 0.5 0.5>))
     Index=&SpawnGroup()
     Num=::NewEnumerator2
    )
    ScaleA=true
    ScaleRGB=true
   )
   &ScaleRibbonWidth(Ribbon Width Scale=&FloatFromCurve(FloatCurve=DataInterface() {propertyValues {Curve {
        defaultValue=1
        keys [{interpMode=RCIM_Cubic, tangentMode=RCTM_User, time=0, value=1}, {interpMode=RCIM_Cubic, tangentMode=RCTM_User, time=1, value=1.5}]
       }}}))
   &ScaleSpriteSize(Uniform Curve Sprite Scale=DataInterface() {propertyValues {Curve {
       defaultValue=0
       keys [
        {arriveTangent=4.1785, interpMode=RCIM_Cubic, leaveTangent=4.1785, tangentMode=RCTM_User, time=0, value=0}
        {arriveTangent=1.1179, interpMode=RCIM_Cubic, leaveTangent=1.1179, tangentMode=RCTM_User, time=0.2042, value=0.6228}
        {arriveTangent=0.0224, interpMode=RCIM_Cubic, leaveTangent=0.0224, tangentMode=RCTM_User, time=1, value=1}
       ]
      }}})
   &ScaleSpriteSize(
    Uniform Curve Index=$Emitter.NormalizedLoopAge
    Uniform Curve Scale=3
    Uniform Curve Sprite Scale=DataInterface() {propertyValues {Curve {defaultValue=0.5, keys [<at 0 0 RCIM_Constant>, {tangentMode=RCTM_User, time=0.003, value=1}, {tangentMode=RCTM_User, time=0.99, value=0}]}}}
   ) {module=ScaleSpriteSize001}
   &SubUVAnimation(
    End Frame Range Override=3
    "Sprite Renderer "=DataInterface() {propertyValues {SpriteRenderer=&self:Stream001_0.Renderer}}
    SubUV Animation Mode:ENiagara_SubUVLookupModeV2=::NewEnumerator2
    UseEndFrame=true
    UseStartFrame=true
   )
   /SetVariables(Particles.FoamDyn:Vector4f=&MakeVector4(
     X=&RandomRangeFloat(Minimum=0, Maximum=0.1)
     Y=0.15
     Z=&RandomRangeFloat(Minimum=0.1, Maximum=0.2)
     W=&FloatFromCurve(
      FloatCurve=DataInterface() {propertyValues {Curve {
         defaultValue=1
         keys [{interpMode=RCIM_Cubic, tangentMode=RCTM_User, time=0, value=0}, {interpMode=RCIM_Cubic, tangentMode=RCTM_User, time=1, value=1}]
        }}}
      Scale Curve=0
     )
    )) {isSetParameters=true}
  ]
  rendererClasses [&NiagaraRibbonRendererProperties, &NiagaraSpriteRendererProperties]
  renderers [
   {rendererClass=&NiagaraRibbonRendererProperties}
   {rendererClass=&NiagaraRibbonRendererProperties, rendererIndex=1}
   {rendererClass=&NiagaraSpriteRendererProperties, rendererIndex=2}
  ]
 }
}
summary {
 systemName=NS_Emote_CrossWiper_Pour
 emitters [
  {emitterName=Stream001, rendererClasses [&NiagaraRibbonRendererProperties, &NiagaraSpriteRendererProperties]}
  {emitterName=SmallStreams001, rendererClasses [&NiagaraRibbonRendererProperties]}
  {emitterName=Droplets003, rendererClasses [&NiagaraSpriteRendererProperties]}
  {emitterName=Droplets004, rendererClasses [&NiagaraSpriteRendererProperties]}
 ]
 userVariables [
  User.Audio Player {
   type=NiagaraDataInterfaceAudioPlayer
   description=""
   defaultValue {
    struct=NiagaraExt_VariableValue_DataInterface
    value {
     dataInterface=&self:NiagaraDataInterfaceAudioPlayer_0
     dataInterfaceClass=&NiagaraDataInterfaceAudioPlayer
    }
   }
  }
  User.PlayWaterSplash {type=NiagaraBool, defaultValue=true}
  User.SplashLocation {type=NiagaraPosition, description="", defaultValue {}}
  User.WaterColor {type=LinearColor, description="", defaultValue {struct=LinearColor, value=<rgba 0.6 0.4745 0.3299 1>}}
 ]
}
