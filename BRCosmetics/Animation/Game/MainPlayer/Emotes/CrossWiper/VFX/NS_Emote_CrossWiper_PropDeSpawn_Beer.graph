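# Niagara effect for the beer prop despawning in the CrossWiper emote.
# Four emitters provide sparkles, airglow, smoke and droplets, with system scripts running System State and several variable assignments.
# The system uses skeletal mesh and color curve data interfaces, plus curve and time-of-day color inputs.
# A NormalizedNotifyProgress user float, defaulting to zero, is exposed, apparently for animation notify timing.
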NiagaraSystem /BRCosmetics/Animation/Game/MainPlayer/Emotes/CrossWiper/VFX/NS_Emote_CrossWiper_PropDeSpawn_Beer
dependencies {
 usedDataInterfaces [&NiagaraDataInterfaceSkeletalMesh, &NiagaraDataInterfaceColorCurve]
 usedDynamicInputs [&ColorFromCurve, &FNTimeOfDayColor]
 usedModules [
  &self:SystemSpawnScript.SystemScriptSource.SystemScriptGraph.NiagaraNodeAssignment_1.SetVariables_#1
  &self:SystemSpawnScript.SystemScriptSource.SystemScriptGraph.NiagaraNodeAssignment_0.SetVariables_#2
  &self:SystemSpawnScript.SystemScriptSource.SystemScriptGraph.NiagaraNodeAssignment_2.SetVariables_#3
  &self:SystemSpawnScript.SystemScriptSource.SystemScriptGraph.NiagaraNodeAssignment_3.SetVariables_#4
  &self:SystemSpawnScript.SystemScriptSource.SystemScriptGraph.NiagaraNodeAssignment_4.SetVariables_#5
  &SystemState
 ]
}
emitters {
 AirglowLWE002 {rendererClasses [], renderers []}
 Droplets {rendererClasses [], renderers []}
 SmokeLWE002 {rendererClasses [], renderers []}
 SparklesLWE002 {rendererClasses [], renderers []}
}
summary {
 systemName=NS_Emote_CrossWiper_PropDeSpawn_Beer
 emitters [
  {emitterName=SparklesLWE002, rendererClasses []}
  {emitterName=AirglowLWE002, rendererClasses []}
  {emitterName=SmokeLWE002, rendererClasses []}
  {emitterName=Droplets, rendererClasses []}
 ]
 userVariables [User.NormalizedNotifyProgress {type=NiagaraFloat, defaultValue=0, description=""}]
}
