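# Niagara water splash effect for the CrossWiper emote, with the visible work done in Splashes_compress; the other four emitters list no renderers.
# Two staggered bursts spawn sprites tagged by spawn group, which drives visibility and per-group switches for lifetime, velocity, offset, size and rotation.
# Sprites use sub-UV animation, gravity, drag, speed-based stretching and per-look dynamic material parameters.
# System splash color is scaled by time of day and faded over life.
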NiagaraSystem /BRCosmetics/Animation/Game/MainPlayer/Emotes/CrossWiper/VFX/NS_Emote_CrossWiper_Splash
dependencies {
 usedDataInterfaces [
  &NiagaraDataInterfaceCurve
  &NiagaraDataInterfaceSpriteRendererInfo
  &NiagaraDataInterfaceColorCurve
 ]
 usedDynamicInputs [
  &MakeFloatFromVector2d
  &RandomRangeVector2D
  &Multiply_Vector2D_ByFloat
  &ColorFromCurve
 ]
 usedModules [
  &self:SystemSpawnScript.SystemScriptSource.SystemScriptGraph.NiagaraNodeAssignment_0.SetVariables_#1
  &self:SystemSpawnScript.SystemScriptSource.SystemScriptGraph.NiagaraNodeAssignment_6.SetVariables_#2
  &self:SystemSpawnScript.SystemScriptSource.SystemScriptGraph.NiagaraNodeAssignment_5.SetVariables_#3
  &self:SystemSpawnScript.SystemScriptSource.SystemScriptGraph.NiagaraNodeAssignment_4.SetVariables_#4
  &self:SystemSpawnScript.SystemScriptSource.SystemScriptGraph.NiagaraNodeAssignment_1.SetVariables_#5
  &self:SystemSpawnScript.SystemScriptSource.SystemScriptGraph.NiagaraNodeAssignment_2.SetVariables_#6
  &self:SystemSpawnScript.SystemScriptSource.SystemScriptGraph.NiagaraNodeAssignment_3.SetVariables_#7
  &SystemState
 ]
}
emitters {
 Droplets {rendererClasses [], renderers []}
 Droplets001 {rendererClasses [], renderers []}
 Splashes {rendererClasses [], renderers []}
 Splashes001 {rendererClasses [], renderers []}
 Splashes_compress {
  emitterUpdateScript [
   &EmitterState(Life Cycle Mode=::NewEnumerator1, Loop Behavior=::NewEnumerator1, Loop Duration=3, UseLoopDelay=false)
   &SpawnBurst_Instantaneous(
    Spawn Count=&UniformRangedInt(Minimum=12, Maximum=15)
    Spawn Group=3
    Spawn Time=0.05
    Use Spawn Probability=false
    UseLoopCountLimit=false
   ) {module=SpawnBurst_Instantaneous003}
   &SpawnBurst_Instantaneous(Spawn Count=&UniformRangedInt(Minimum=12, Maximum=15), Use Spawn Probability=false)
  ]
  particleSpawnScript [
   /SetVariables(Particles.VisibilityTag=&SpawnGroup()) {isSetParameters=true}
   /SetVariables(Particles.Scaler:NiagaraFloat=&FloatFromCurve(
     FloatCurve=DataInterface() {propertyValues {Curve {keys [
         <at 0 0 RCIM_Cubic>
         {arriveTangent=2.9389, interpMode=RCIM_Cubic, leaveTangent=2.9389, tangentMode=RCTM_User, time=0.5, value=0.5}
         <at 1 1 RCIM_Cubic>
        ]}}}
     CurveIndex=&RandomRangeFloat(Minimum=0, Maximum=1)
    )) {isSetParameters=true}
   &InitializeParticle(
    Color=$System.Water Splashes color
    Lifetime=&MergeHelper_FloatSwitch(
     0=&RandomRangeFloat(Minimum=0.3, Maximum=0.6)
     1=&RandomRangeFloat(Minimum=0.3, Maximum=0.6)
     2=&RandomRangeFloat(Minimum=0.2, Maximum=0.55)
     3=&RandomRangeFloat(Minimum=0.3, Maximum=0.6)
     4=&RandomRangeFloat(Minimum=0.45, Maximum=0.65)
     5=&RandomRangeFloat(Minimum=0.65, Maximum=0.85)
     Index=&SpawnGroup()
     Num=::NewEnumerator4
    )
    Sprite Rotation Angle=&MergeHelper_FloatSwitch(
     0=&RandomRangeFloat(Minimum=0, Maximum=1)
     1=&RandomRangeFloat(Minimum=0, Maximum=1)
     2=&RandomRangeFloat(Minimum=0, Maximum=0.5)
     3=&RandomRangeFloat(Minimum=0, Maximum=0.2)
     4=&RandomRangeFloat(Minimum=0, Maximum=1)
     5=&RandomRangeFloat(Minimum=0, Maximum=0.2)
     Index=&SpawnGroup()
     Num=::NewEnumerator4
    )
    Sprite Rotation Mode=::NewEnumerator1
    Sprite Size Max=<9 18>
    Sprite Size Min=<7 12>
    Sprite Size Mode=::NewEnumerator2
    Sprite UV Mode=::NewEnumerator3
    UsePositionOffset=false
    Write Material Random=true
   )
   &AddVelocity(Velocity=&MergeHelper_VectorSwitch(
     0=&RandomRangeVector(Minimum=<50 -25 150>, Maximum=<225 75 300>)
     1=&RandomRangeVector(Minimum=<0 0 10>, Maximum=<0 0 30>)
     2=&RandomRangeVector(Minimum=<-5 -5 10>, Maximum=<5 5 20>)
     3=&RandomRangeVector(Minimum=<-50 -25 50>, Maximum=<-175 75 160>)
     4=&RandomRangeVector(Minimum=<-50 -50 75>, Maximum=<50 50 250>)
     5=&RandomRangeVector(Minimum=<-50 -50 200>, Maximum=<50 50 300>)
     Index=&SpawnGroup()
     Num=::NewEnumerator4
    ))
   &ShapeLocation(
    MaskSpawnGroup=false
    Offset=&MergeHelper_VectorSwitch(
     0=<0 0 0>
     1=<0 0 -9.022>
     2=<0 0 0>
     3=<0 0 0>
     4=<0 0 10>
     5=<0 0 0>
     Index=&SpawnGroup()
     Num=::NewEnumerator4
    )
    Offset Mode=::NewEnumerator0
    Override Seed=false
    Sphere Radius=&MergeHelper_FloatSwitch(0=3, 1=15, 2=3, 3=3, 4=1, 5=10, Index=&SpawnGroup(), Num=::NewEnumerator4)
    Sphere Surface Distribution=0.4357
    UseOwnerScale=true
   )
   &SubUVAnimation(
    End Frame Range Override=3
    Evaluation Type=::NewEnumerator0
    "Sprite Renderer "=DataInterface() {propertyValues {SpriteRenderer=&self:Splashes_compress_0.NiagaraSpriteRendererProperties_0}}
    SubUV Animation Mode:ENiagara_SubUVLookupModeV2=::NewEnumerator2
    UseEndFrame=true
    UseStartFrame=true
   )
   &SubUVAnimation(
    End Frame Range Override=4
    Evaluation Type=::NewEnumerator0
    "Sprite Renderer "=DataInterface() {propertyValues {SpriteRenderer=&self:Splashes_compress_0.NiagaraSpriteRendererProperties_2}}
    SubUV Animation Mode:ENiagara_SubUVLookupModeV2=::NewEnumerator2
    UseEndFrame=true
    UseStartFrame=true
   ) {module=SubUVAnimation001}
   &SubUVAnimation(
    End Frame Range Override=3
    Evaluation Type=::NewEnumerator0
    "Sprite Renderer "=DataInterface() {propertyValues {SpriteRenderer=&self:Splashes_compress_0.NiagaraSpriteRendererProperties_7}}
    SubUV Animation Mode:ENiagara_SubUVLookupModeV2=::NewEnumerator2
    UseEndFrame=true
    UseStartFrame=true
   ) {module=SubUVAnimation003}
   &SubUVAnimation(
    End Frame Range Override=3
    Evaluation Type=::NewEnumerator0
    "Sprite Renderer "=DataInterface() {propertyValues {SpriteRenderer=&self:Splashes_compress_0.NiagaraSpriteRendererProperties_9}}
    SubUV Animation Mode:ENiagara_SubUVLookupModeV2=::NewEnumerator2
    UseEndFrame=true
    UseStartFrame=true
   ) {module=SubUVAnimation004}
   &CameraOffset(Camera Offset Amount=&MergeHelper_FloatSwitch(
     0=&RandomRangeFloat(Minimum=0, Maximum=10)
     1=&RandomRangeFloat(Minimum=6, Maximum=15)
     2=&RandomRangeFloat(Minimum=6, Maximum=20)
     3=&RandomRangeFloat(Minimum=0, Maximum=3)
     4=0
     5=&RandomRangeFloat(Minimum=0, Maximum=3)
     Index=&SpawnGroup()
     Num=::NewEnumerator4
    ))
   &ApplyInitialForces()
  ]
  particleUpdateScript [
   &ParticleState()
   &ScaleSpriteSize(
    Scale Sprite Size Mode=::NewEnumerator0
    Uniform Scale Factor=&MergeHelper_FloatSwitch(
     0=&FloatFromCurve(
      FloatCurve=DataInterface() {propertyValues {Curve {
         defaultValue=0
         keys [
          {arriveTangent=4.033, interpMode=RCIM_Cubic, leaveTangent=4.033, tangentMode=RCTM_User, time=0, value=0}
          {interpMode=RCIM_Cubic, tangentMode=RCTM_User, time=0.75, value=1}
          {tangentMode=RCTM_User, time=1, value=0}
         ]
        }}}
      Scale Curve=1.25
     )
     1=&FloatFromCurve(FloatCurve=DataInterface() {propertyValues {Curve {keys [
          {arriveTangent=4.1785, interpMode=RCIM_Cubic, leaveTangent=4.1785, tangentMode=RCTM_User, time=0, value=0.15}
          {arriveTangent=1.1179, interpMode=RCIM_Cubic, leaveTangent=1.1179, tangentMode=RCTM_User, time=0.2042, value=0.6228}
          {arriveTangent=0.0224, interpMode=RCIM_Cubic, leaveTangent=0.0224, tangentMode=RCTM_User, time=1, value=1}
         ]}}})
     2=&FloatFromCurve(FloatCurve=DataInterface() {propertyValues {Curve {keys [
          {arriveTangent=0.1295, interpMode=RCIM_Cubic, leaveTangent=0.1295, tangentMode=RCTM_User, time=0, value=0}
          {arriveTangent=1, interpMode=RCIM_Cubic, leaveTangent=1, time=0.35, value=0.9}
          <at 1 1 RCIM_Cubic>
         ]}}})
     3=&FloatFromCurve(FloatCurve=DataInterface() {propertyValues {Curve {
         defaultValue=0
         keys [
          {arriveTangent=4.033, interpMode=RCIM_Cubic, leaveTangent=4.033, tangentMode=RCTM_User, time=0, value=0}
          {interpMode=RCIM_Cubic, tangentMode=RCTM_User, time=0.75, value=1}
          {tangentMode=RCTM_User, time=1, value=0}
         ]
        }}})
     4=&Lerp_Float(
      A=&FloatFromCurve(FloatCurve=DataInterface() {propertyValues {Curve {keys [
           {arriveTangent=2.9966, interpMode=RCIM_Cubic, leaveTangent=2.9966, tangentMode=RCTM_User, time=0, value=0}
           {arriveTangent=1, interpMode=RCIM_Cubic, leaveTangent=1, time=0.3, value=0.6}
           {arriveTangent=0.5346, interpMode=RCIM_Cubic, leaveTangent=0.5346, tangentMode=RCTM_User, time=1, value=1}
          ]}}})
      B=&FloatFromCurve(FloatCurve=DataInterface() {propertyValues {Curve {keys [
           {arriveTangent=1.5909, interpMode=RCIM_Cubic, leaveTangent=1.5909, tangentMode=RCTM_User, time=0, value=0}
           {arriveTangent=0.5346, interpMode=RCIM_Cubic, leaveTangent=0.5346, tangentMode=RCTM_User, time=1, value=1}
          ]}}})
      Alpha=$Particles.Scaler
     )
     5=&FloatFromCurve(FloatCurve=DataInterface() {propertyValues {Curve {keys [
          {arriveTangent=4.1785, interpMode=RCIM_Cubic, leaveTangent=4.1785, tangentMode=RCTM_User, time=0, value=0}
          {arriveTangent=1.1179, interpMode=RCIM_Cubic, leaveTangent=1.1179, tangentMode=RCTM_User, time=0.2042, value=0.6228}
          {arriveTangent=0.0224, interpMode=RCIM_Cubic, leaveTangent=0.0224, tangentMode=RCTM_User, time=1.0045, value=0.824}
         ]}}})
     Index=&SpawnGroup()
     Num=::NewEnumerator4
    )
   )
   &SpriteRotationRate(Rotation Rate=&MergeHelper_FloatSwitch(
     0=&Lerp_Float(A=70, B=150, Alpha=$Particles.Initial.MaterialRandom)
     1=&RandomRangeFloat(Minimum=-70, Maximum=-100)
     2=&RandomRangeFloat(Minimum=-50, Maximum=-150)
     3=&RandomRangeFloat(Minimum=-70, Maximum=-150)
     4=&RandomRangeFloat(Minimum=-0.5, Maximum=0.5)
     5=&RandomRangeFloat(Minimum=-70, Maximum=-100)
     Index=&SpawnGroup()
     Num=::NewEnumerator4
    ))
   &ScaleColor(
    Scale Alpha=&FloatFromCurve(
     FloatCurve=DataInterface() {propertyValues {Curve {keys [
         {arriveTangent=-0.2723, interpMode=RCIM_Cubic, leaveTangent=14.5791, tangentMode=RCTM_Break, time=0, value=0}
         {interpMode=RCIM_Cubic, tangentMode=RCTM_Break, time=0.1, value=1}
         {arriveTangent=0.0378, interpMode=RCIM_Cubic, leaveTangent=0.0378, tangentMode=RCTM_User, time=0.3785, value=1.0282}
         {arriveTangent=-3.7547, interpMode=RCIM_Cubic, leaveTangent=-3.7547, tangentMode=RCTM_Break, time=0.7, value=0}
        ]}}}
     Scale Curve=&FNTimeOfDayScalar(Night Scale=0.7, Morning Scale=0.5, Day Scale=0.6, Evening Scale=0.5)
    )
    Scale RGB=&MakeVectorFromLinearColorRGB(Color=&FNTimeOfDayColor(Night Color=<rgba 1 1 1 0.7>, Morning Color=<rgba 1 1 1 0.6>, Day Color=<rgba 0.7 0.7 0.7 1>, Evening Color=<rgba 0.5 0.5 0.5 0.5>))
    ScaleA=true
    ScaleRGB=true
   )
   &GravityForce(Gravity=<0 0 -550>)
   &Drag(
    Drag=&RandomRangeFloat(Minimum=1, Maximum=1)
    Rotational Drag:NiagaraFloat=&MergeHelper_FloatSwitch(0=1, 1=1, 2=1, 3=1, 4=1, Index=&SpawnGroup(), Num=::NewEnumerator3)
    Use Linear Drag=true
    Use Rotational Drag=true
   )
   &AccelerationForce(Acceleration=&MergeHelper_VectorSwitch(
     0=&UniformRangedVector(Minimum=<0 0 -50>, Maximum=<0 0 -500>)
     1=<0 0 0>
     2=<0 0 0>
     3=<0 0 0>
     4=<0 0 0>
     5=<1 1 1>
     Index=&SpawnGroup()
     Num=::NewEnumerator4
    ))
   &SolveForcesAndVelocity(Clamp Velocity=false, Limit Acceleration=false)
   &ScaleSpriteSizeBySpeed(
    Max Scale Factor=&MergeHelper_Vector2Switch(0=<1 2>, 1=<2 5>, 2=<3 5>, 3=<1 2>, 4=<2 4>, 5=<2 2>, Index=&SpawnGroup(), Num=::NewEnumerator4)
    Min Scale Factor=&MergeHelper_Vector2Switch(0=<2 1>, 1=<1 4>, 2=<2 1>, 3=<2 1>, 4=<1 1>, 5=<1 1>, Index=&SpawnGroup(), Num=::NewEnumerator4)
    Sample Scale Factor By Curve=true
    Scale Factor Curve=DataInterface() {propertyValues {Curve {defaultValue=0.5941, keys [{tangentMode=RCTM_User, time=0, value=1}, {tangentMode=RCTM_User, time=1, value=0}]}}}
    Velocity Threshold:NiagaraFloat=&MergeHelper_FloatSwitch(0=250, 1=250, 2=1000, 3=250, 4=250, 5=250, Index=&SpawnGroup(), Num=::NewEnumerator4)
   )
   /SetVariables(
    Particles.0 - Water Emerge DynParam:Vector4f=&MakeVector4(
     X=&RandomRangeFloat(Minimum=0, Maximum=0.2)
     Y=&RandomRangeFloat(Minimum=0, Maximum=0.2)
     Z=&RandomRangeFloat(Minimum=0.5, Maximum=0.8)
     W=&FloatFromCurve(FloatCurve=DataInterface() {propertyValues {Curve {
         defaultValue=1
         keys [
          <at 0 0>
          {arriveTangent=1, interpMode=RCIM_Cubic, leaveTangent=1, time=0.1, value=0}
          {arriveTangent=3.4729, interpMode=RCIM_Cubic, leaveTangent=3.4729, tangentMode=RCTM_User, time=1, value=1}
         ]
        }}})
    )
    Particles.1 - water crown DynParam:Vector4f=&MakeVector4(
     X=&RandomRangeFloat(
      Minimum=&FloatFromCurve(FloatCurve=DataInterface() {propertyValues {Curve {keys [
           {interpMode=RCIM_Cubic, tangentMode=RCTM_User, time=0, value=1}
           {arriveTangent=-3.7889, interpMode=RCIM_Cubic, leaveTangent=-3.7889, tangentMode=RCTM_User, time=0.75, value=0}
          ]}}})
      Maximum=&FloatFromCurve(FloatCurve=DataInterface() {propertyValues {Curve {keys [
           {interpMode=RCIM_Cubic, tangentMode=RCTM_User, time=0, value=1}
           {arriveTangent=-3.7889, interpMode=RCIM_Cubic, leaveTangent=-3.7889, tangentMode=RCTM_User, time=1, value=0}
          ]}}})
     )
     Y=0
     Z=0
     W=0
    )
    Particles.2 - Mist DynParam:Vector4f=&MakeVector4(
     X=&FloatFromCurve(FloatCurve=DataInterface() {propertyValues {Curve {keys [{arriveTangent=-0.8567, interpMode=RCIM_Cubic, leaveTangent=-0.8567, tangentMode=RCTM_User, time=0, value=0}, <at 1 -0.4 RCIM_Cubic>]}}})
     Y=&RandomRangeFloat(Minimum=0, Maximum=10)
     Z=0
     W=&RandomRangeFloat(Minimum=0.1, Maximum=0.4)
    )
    Particles.3 - Splash Ring DynParam:Vector4f=&MakeVector4(
     X=&RandomRangeFloat(Minimum=0, Maximum=0.2)
     Y=&RandomRangeFloat(Minimum=0, Maximum=0.2)
     Z=&RandomRangeFloat(Minimum=0.5, Maximum=0.8)
     W=&FloatFromCurve(FloatCurve=DataInterface() {propertyValues {Curve {
         defaultValue=1
         keys [
          {interpMode=RCIM_Cubic, tangentMode=RCTM_User, time=0, value=0}
          {arriveTangent=3.4729, interpMode=RCIM_Cubic, leaveTangent=3.4729, tangentMode=RCTM_User, time=1, value=1}
         ]
        }}})
    )
    Particles.4 - Suplemental splash DynParam:Vector4f=&MakeVector4(
     X=&FloatFromCurve(FloatCurve=DataInterface() {propertyValues {Curve {keys [
          {interpMode=RCIM_Cubic, tangentMode=RCTM_User, time=0, value=1}
          {arriveTangent=-3.7889, interpMode=RCIM_Cubic, leaveTangent=-3.7889, tangentMode=RCTM_User, time=0.8, value=0}
         ]}}})
     Y=0
     Z=0
     W=0
    )
    Particles.5 - end splash DynParam:Vector4f=&MakeVector4(
     X=&FloatFromCurve(FloatCurve=DataInterface() {propertyValues {Curve {keys [{interpMode=RCIM_Cubic, tangentMode=RCTM_Break, time=0, value=1}, {interpMode=RCIM_Cubic, tangentMode=RCTM_Break, time=0.5, value=0.4}]}}})
     Y=0
     Z=0
     W=0
    )
   ) {isSetParameters=true}
   /SetVariables(Particles.1 - stretch water crown velocity:Vector3f=&Lerp_Vector(
     A=&NormalizeVector(
      Vector To Normalize=&Add_Vector(A=$Particles.Initial.ShapeLocation.ShapeNormal, B=<0 0 3>)
      Safe Normalize=true
      Safe Fallback Threshold=0
      Safe Fallback Vector=<0 0 0.001>
     )
     B=&NormalizeVector(
      Vector To Normalize=&Add_Vector(A=$Particles.Initial.ShapeLocation.ShapeNormal, B=<0 0 1>)
      Safe Normalize=true
      Safe Fallback Threshold=0
      Safe Fallback Vector=<0 0 0.001>
     )
     Alpha=&FloatFromCurve(FloatCurve=DataInterface() {propertyValues {Curve {keys [
          {interpMode=RCIM_Cubic, tangentMode=RCTM_User, time=0, value=0}
          {arriveTangent=2.1818, interpMode=RCIM_Cubic, leaveTangent=2.1818, tangentMode=RCTM_User, time=1, value=1}
         ]}}})
    )) {enabled=false, isSetParameters=true}
  ]
  rendererClasses [&NiagaraSpriteRendererProperties]
  renderers [
   {rendererClass=&NiagaraSpriteRendererProperties}
   {rendererClass=&NiagaraSpriteRendererProperties, rendererIndex=1}
  ]
 }
}
summary {
 systemName=NS_Emote_CrossWiper_Splash
 emitters [
  {emitterName=Droplets, rendererClasses []}
  {emitterName=Splashes, rendererClasses []}
  {emitterName=Splashes001, rendererClasses []}
  {emitterName=Droplets001, rendererClasses []}
  {emitterName=Splashes_compress, rendererClasses [&NiagaraSpriteRendererProperties]}
 ]
 userVariables [
  User.AnimCurve_FXdelay {type=NiagaraFloat, defaultValue=0, description=""}
  User.bIsFrontEnd {type=NiagaraBool}
  User.bIsFrontEndPreview {type=NiagaraBool}
  User.NormalizedNotifyProgress {type=NiagaraFloat, defaultValue=0, description=""}
 ]
}
